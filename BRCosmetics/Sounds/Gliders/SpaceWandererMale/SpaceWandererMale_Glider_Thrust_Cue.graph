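# Sound cue for the thrust sound of the SpaceWandererMale cosmetic glider.
# A Random node picks one of three thrust wave variations with equal weight.
# Playback uses the Athena player glider thrust attenuation, the cosmetic glider loop audio properties sheet and the gliders land sound class, at reduced volume.
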SoundCue /BRCosmetics/Sounds/Gliders/SpaceWandererMale/SpaceWandererMale_Glider_Thrust_Cue
root=Random_0
nodes [
 {id=Random_0, children [WavePlayer_0,WavePlayer_1,WavePlayer_2], properties {weights [1,1,1]}}
 {id=WavePlayer_0, properties {soundWaveAssetPtr=&/SpaceWandererMale_Glider_Thrust_01}}
 {id=WavePlayer_1, properties {soundWaveAssetPtr=&/SpaceWandererMale_Glider_Thrust_02}}
 {id=WavePlayer_2, properties {soundWaveAssetPtr=&/SpaceWandererMale_Glider_Thrust_03}}
]
properties {
 attenuationSettings=&Fort_Player_Glider_Thrust_Athena_Att
 audioPropertiesSheet=&APS_Cosmetic_Glider_Loop
 maxDistance=12500
 soundClassObject=&Fort_Gliders_Land
 volumeMultiplier=0.5
}
